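# M_ShineSpike is a two-sided additive unlit mesh-particle material for glowing, animated spike effects in the Nexus environment.
# Panned electric noise multiplies curve-atlas ramps, tint and particle color to form emissive, with depth fade and optional edge mask for opacity.
# A noise-jittered directional offset, widened along V and pushed by CameraOffset, deforms the spikes.
# Motion, color, exponent and scale blend from MPC_Nexus values to dynamic or local parameters via one switch scalar.
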Material /Game/Athena/Environments/Nexus/Materials/FXshader/M_ShineSpike
graph {
 properties {
  bUsedWithMeshParticles=true
  bUsedWithNiagaraMeshParticles=true
  blendMode=BLEND_Additive
  pixelDepthOffsetMode=PDOM_Legacy
  shadingModel=MSM_Unlit
  twoSided=true
  usedShadingModels=Unlit
 }
 nodes [
  n1 = DynamicParameter() {parameterIndex=1, defaultValue=<rgba 1 1 1 0>, paramNames [ColorExp, MontionStrength, Scale, Desaturation]}
  n2 = ScalarParameter() {group=Param, parameterName=Change MPC to Dy+Param}
  n3 = LinearInterpolate(
   A=CollectionParameter() {collection=&/MPC_Nexus, parameterName=Spike_MotionStrength}
   B=@n1.MontionStrength
   Alpha=@n2
  )
  n4 = Add(A=@n3, B=DynamicParameter() {paramNames [Time,Random,Param3,Param4]}.Random)
  n5 = TextureCoordinate()
  n6 = ComponentMask(None=@n5) {r=true}
  n7 = Add(A=@n6, B=Multiply(A=@n4, B=ScalarParameter() {defaultValue=-2.3, parameterName=WPOTimeScale}))
  n8 = Multiply(A=@n7) {constB=2}
  n9 = ComponentMask(None=TextureCoordinate() {vTiling=0.01}) {g=true}
  n10 = TextureSample(UVs=Panner(Coordinate=AppendVector(A=@n8, B=@n9), Time=@n4, Speed=Constant2Vector() {g=0.0051, r=0.05})) {samplerType=SAMPLERTYPE_LinearColor, texture=/Game/Characters/Player/Male/Medium/Bodies/M_MED_SRDrift/Textures/T_ElectricNoise}
  n11 = TextureCoordinate()
  n12 = CurveAtlasRowParameter(CurveTime=ComponentMask(None=@n11) {r=true}) {atlas=/Game/Athena/LinearColorCurves/TempAthenaCurveAtlas, curve=&/Curve_Ramp2}
  n13 = CurveAtlasRowParameter(CurveTime=@n12.B) {
   atlas=&AthenaCurveAtlas01
   curve=&/Curve_Color
   defaultValue=78
   parameterName=Param2
  }
  n14 = CurveAtlasRowParameter(CurveTime=@n12.A) {
   atlas=&AthenaCurveAtlas01
   curve=&Curve_Color
   defaultValue=78
   parameterName=Param2
  }
  n15 = VectorParameter() {group=Param, parameterName=ColorTint, defaultValue=<rgba 0.1639 0.4298 0.5 1>}
  n16 = DynamicParameter() {parameterIndex=1, defaultValue=<rgba 1 1 1 0>, paramNames [ColorExp, MontionStrength, Scale, Desaturation]}
  n17 = LinearInterpolate(
   A=CollectionParameter() {collection=&MPC_Nexus, parameterName=Spike_ColorOvrride}
   B=AppendVector(A=Desaturation(None=@n15.RGB, Fraction=@n16.Desaturation), B=@n15.A)
   Alpha=@n2
  )
  n18 = CurveAtlasRowParameter(CurveTime=ComponentMask(None=@n11) {g=true}) {atlas=&TempAthenaCurveAtlas, curve=&Curve_Ramp2}
  n19 = Multiply(
   A=Multiply(A=Add(A=@n13, B=@n14), B=ComponentMask(None=@n17) {b=true, g=true, r=true})
   B=Multiply(A=Multiply(A=@n18.R, B=@n12.G), B=ComponentMask(None=@n17) {a=true})
  )
  n20 = ParticleColor()
  n21 = Multiply(A=@n10.R, B=Multiply(A=@n19, B=@n20.RGB))
  n22 = DynamicParameter() {parameterIndex=1, defaultValue=<rgba 1 1 1 0>, paramNames [ColorExp, MontionStrength, Scale, Desaturation]}
  n23 = LinearInterpolate(
   A=CollectionParameter() {collection=&MPC_Nexus, parameterName=Spike_ColorExp}
   B=@n22.ColorExp
   Alpha=@n2
  )
  n24 = ScalarParameter() {defaultValue=-2, parameterName=Sat}
  n25 = Add(A=Max(A=Power(Base=@n21, Exp=@n23), B=Desaturation(None=@n21, Fraction=@n24)), B=ScalarParameter() {group=Final, parameterName=Final Add})
  n26 = ScalarParameter() {group=Final, parameterName=Final Desaturation}
  MaterialFunctionCall(Emissive=Multiply(A=Desaturation(None=@n25, Fraction=@n26), B=VectorParameter() {group=Final, parameterName=Final Color, defaultValue=<rgba 1 1 1 0>}.RGB)) {materialFunction=&MF_FX_EmissiveAdjustments}
  n27 = CurveAtlasRowParameter(CurveTime=@n6) {
   atlas=&TempAthenaCurveAtlas
   curve=&/NewCurveBase3
   defaultValue=2
  }
  n28 = Time()
  n29 = ScalarParameter() {defaultValue=2, parameterName=PluseSpeed}
  n30 = CurveAtlasRowParameter(CurveTime=Frac(None=Multiply(A=@n28, B=@n29))) {
   atlas=&TempAthenaCurveAtlas
   curve=&NewCurveBase3
   defaultValue=0.0098
  }
  Multiply(A=@n27.R, B=@n30.G)
  n31 = ObjectPositionWS()
  n32 = WorldPosition()
  n33 = RotateAboutAxis(
   NormalizedRotationAxis=Transform(None=Constant3Vector() {constant=<rgba 0 0 1 0>}) {transformSourceType=TRANSFORMSOURCE_Local}
   RotationAngle=@n4
   PivotPoint=@n31
   Position=@n32.XYZ
  )
  n34 = Normalize(VectorInput=Subtract(A=@n31, B=Add(A=@n33, B=@n32.XYZ)))
  n35 = AppendVector(A=@n7, B=Constant())
  n36 = Multiply(
   A=Subtract(A=TextureSample(UVs=@n35) {samplerType=SAMPLERTYPE_LinearColor, texture=&HighContrastNoise}.RGB) {constB=0.5}
   B=LinearInterpolate(
    A=ScalarParameter() {defaultValue=0.5, parameterName=Offset1}
    B=ScalarParameter() {defaultValue=1, parameterName=Offset2}
    Alpha=ComponentMask(None=@n5) {g=true}
   )
  )
  n37 = ScalarParameter() {defaultValue=1050, parameterName=Wide1}
  n38 = DynamicParameter() {parameterIndex=1, defaultValue=<rgba 1 1 1 0>, paramNames [ColorExp, MontionStrength, Scale, Desaturation]}
  n39 = Multiply(
   A=Multiply(A=ScalarParameter() {defaultValue=2000, parameterName=Wide2}, B=@n27.R)
   B=LinearInterpolate(
    A=CollectionParameter() {collection=&MPC_Nexus, parameterName=Spike_Scale}
    B=@n38.Scale
    Alpha=@n2
   )
  )
  n40 = TextureCoordinate()
  n41 = Multiply(A=Normalize(VectorInput=Add(A=@n34, B=@n36)), B=LinearInterpolate(A=@n37, B=Add(A=@n37, B=@n39), Alpha=ComponentMask(None=@n40) {g=true}))
  n42 = Add(
   A=@n41
   B=MaterialFunctionCall(Offset Amount=ScalarParameter() {defaultValue=500, parameterName=Offset}, World Position=@n41) {materialFunction=/Engine/Functions/Engine_MaterialFunctions02/WorldPositionOffset/CameraOffset}.Clamped Camera Offset
  )
  n43 = Subtract(A=Add(A=WorldPosition().XYZ, B=@n42), B=CameraPositionWS())
  n44 = Abs(None=DotProduct(A=Normalize(VectorInput=@n43), B=CameraVectorWS()))
  Multiply(A=@n21, B=@n44)
  n45 = DynamicParameter() {paramNames [Time,Random,Param3,Param4]}
  Multiply(A=Add(A=@n45.Time, B=@n45.Time))
  Multiply(A=@n4, B=Multiply(A=@n3) {constB=10})
  n46 = ScalarParameter() {defaultValue=200, parameterName=DF_FD}
  n47 = Multiply(A=@n10.R, B=@n19)
  n48 = LinearInterpolate(A=@n20.A, B=Multiply(A=Desaturation(None=@n47), B=@n20.A), Alpha=ScalarParameter() {parameterName=Use TexMask})
  n49 = MaterialFunctionCall() {materialFunction=&LinearGradient}
  n50 = OneMinus(None=@n49.UGradient)
  n51 = Multiply(A=Power(Base=Multiply(A=@n49.UGradient, B=@n50)), B=OneMinus(None=@n49.VGradient))
  n52 = ScalarParameter() {parameterName=Ues Edge Mask}
  Multiply(
   A=ShadingPathSwitch(Default=DepthFade(Opacity=@n48, FadeDistance=@n46), Mobile=@n48) {desc=Migrated from FeatureLevelSwitch}
   B=Multiply(A=ScalarParameter() {defaultValue=1, group=Final, parameterName=Final Mask}, B=LinearInterpolate(B=@n51, Alpha=@n52) {constA=1})
  )
  OneMinus(None=@n44)
  Reroute(None=@n42)
  ScalarParameter() {defaultValue=1, group=Param, parameterName=Spike_Scale}
  ScalarParameter() {defaultValue=1, group=Param, parameterName=Spike_MontionStrength}
  ScalarParameter() {defaultValue=0.2, parameterName=TimeDebugScale}
  ScalarParameter() {defaultValue=1, parameterName=ColorExp}
  Time()
 ]
}
