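# Translucent, two-sided unlit material for Zero Point solar flare static meshes in the Nexus environment.
# Two panned magnetic-wave samples plus tinted caustics, offset by vertex color, combine with a masked color tint to drive emissive and power-shaped opacity.
# Vertex color varies SplineThicken width, while dynamic-parameter Squeeze adds world position offset and alters tiling.
# Opacity uses view-angle, camera and depth fades, simplified on mobile, with optional time-of-day scaling.
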Material /Game/Athena/Environments/Nexus/Materials/FXshader/M_ZeroPoint_SolarFlare_StaticMesh
graph {
 properties {blendMode=BLEND_Translucent, pixelDepthOffsetMode=PDOM_Legacy, shadingModel=MSM_Unlit, twoSided=true, usedShadingModels=Unlit}
 nodes [
  n1 = VertexColor()
  n2 = LinearInterpolate(
   A=ScalarParameter() {defaultValue=25, parameterName=WidthBase_Min}
   B=ScalarParameter() {defaultValue=25, parameterName=WidthBase_Max}
   Alpha=@n1.R
  )
  n3 = Constant() {r=1}
  n4 = QualitySwitch(Default=ShadingPathSwitch(Default=@n3), Low=Constant() {r=0.5}, High=@n3, Medium=@n3, Epic=@n3)
  n5 = LinearInterpolate(A=ScalarParameter() {parameterName=WidthTip_Min}, B=ScalarParameter() {parameterName=WidthTip_Max}, Alpha=@n1.G)
  n6 = DynamicParameter() {defaultValue=<rgba 1 1 12 0>, paramNames [Random,Erode,CenterPower,Squeeze]}
  n7 = Multiply(A=@n6.Squeeze, B=ScalarParameter() {defaultValue=-256, parameterName=SqueezeWPOMult})
  n8 = Subtract(A=ComponentMask(None=TextureCoordinate()) {r=true}) {constB=0.5}
  n9 = Transform(None=AppendVector(A=Multiply(A=@n8) {constB=2}, B=Constant2Vector()))
  Add(
   A=MaterialFunctionCall(WidthBase=Multiply(A=@n2, B=@n4), WidthTip=Multiply(A=@n5, B=@n4)) {materialFunction=&SplineThicken}.WorldPositionOffset
   B=LinearInterpolate(A=Constant(), B=Multiply(A=@n7, B=@n9), Alpha=ScalarParameter() {defaultValue=1, parameterName=Use WPO})
  )
  Constant() {r=2}
  MaterialFunctionCall() {materialFunction=&DebugTimeSine}
  n10 = AppendVector(A=Constant(), B=VertexColor().G)
  n11 = TextureCoordinate()
  n12 = VectorParameter() {parameterName=Tile1, defaultValue=<rgba 0.4 0.5 0 0>}
  n13 = AppendVector(A=@n12.R, B=@n12.R)
  n14 = VectorParameter() {parameterName=Tile1SqueezeMult, defaultValue=<rgba 0.3 2.5 0 1>}
  n15 = LinearInterpolate(A=@n13, B=Multiply(A=@n13, B=AppendVector(A=@n14.R, B=@n14.R)), Alpha=@n6.Squeeze) {constA=0.8, constB=2}
  n16 = Multiply(A=Add(A=Add(A=@n10, B=@n11), B=Multiply(B=@n15) {constA=-0.2, constB=0.5}), B=@n15)
  n17 = VectorParameter() {parameterName=PanDirection1, defaultValue=<rgba 0.2 0.02 0 0>}
  n18 = AppendVector(A=@n17.R, B=@n17.R)
  n19 = Add(A=Multiply(A=Time(), B=ScalarParameter() {defaultValue=0.5, parameterName=PanSpeed}), B=VertexColor().R)
  n20 = Panner(Coordinate=Multiply(A=@n16) {constB=1.5}, Time=@n19, Speed=@n18) {speedX=-0.2, speedY=-0.02}
  n21 = VectorParameter() {parameterName=PanDirection1Mult, defaultValue=<rgba 1 -1 0 1>}
  n22 = Panner(Coordinate=@n16, Time=@n19, Speed=Multiply(A=@n18, B=AppendVector(A=@n21.R, B=@n21.R))) {speedX=-0.2, speedY=0.02}
  n23 = AppendVector(A=Constant(), B=VertexColor().G)
  n24 = TextureCoordinate()
  n25 = VectorParameter() {parameterName=Tile2, defaultValue=<rgba 0.3 0.5 0 0>}
  n26 = VectorParameter() {parameterName=PanDirection2, defaultValue=<rgba 0.1 0.04 0 0>}
  n27 = Panner(Coordinate=Multiply(A=Add(A=@n23, B=@n24), B=AppendVector(A=@n25.R, B=@n25.R)), Time=@n19, Speed=AppendVector(A=@n26.R, B=@n26.R)) {speedX=0.1, speedY=0.04}
  n28 = Multiply(
   A=Constant3Vector() {constant=<rgba 1 0 0.4 0>}
   B=TextureSample(UVs=Add(A=@n27) {constB=0.01}) {texture=/Game/Global/Textures/Default/Caustics/T_Caustics_Softer}.G
  )
  n29 = Multiply(
   A=Constant3Vector() {constant=<rgba 0 1 0.6 0>}
   B=TextureSample(UVs=@n27) {texture=&T_Caustics_Softer}.G
  )
  n30 = Add(
   A=Add(
    A=TextureSample(UVs=@n20) {texture=/Game/Athena/Prototype/Blueprints/Sneezy/Textures/T_Sneezy_MagneticWaves}.RGB
    B=TextureSample(UVs=@n22) {texture=&T_Sneezy_MagneticWaves}.RGB
   )
   B=Multiply(A=LinearInterpolate(A=@n28, B=@n29))
  )
  n31 = VectorParameter() {parameterName=Color, defaultValue=<rgba 0.0771 0.2812 0.4 0>}
  n32 = VectorParameter() {parameterName=Color_Low_And_Mobile, defaultValue=<rgba 0.0771 0.2812 0.4 0>}
  n33 = Constant() {r=0.5}
  n34 = Power(Base=SphereMask(A=TextureCoordinate(), B=@n33, Radius=@n33) {hardnessPercent=0})
  n35 = Multiply(
   A=QualitySwitch(Default=ShadingPathSwitch(Default=@n31.RGB, Forward=@n32.RGB, Mobile=@n32.RGB), Low=@n32.RGB, High=@n31.RGB, Medium=@n31.RGB, Epic=@n31.RGB)
   B=@n34
  )
  MaterialFunctionCall(Emissive=Add(A=@n30, B=@n35)) {materialFunction=&MF_FX_EmissiveAdjustments}
  n36 = ComponentMask(None=TextureCoordinate()) {g=true}
  n37 = ScalarParameter() {defaultValue=0.5, parameterName=ErosionCenter}
  n38 = LinearInterpolate(B=ScalarParameter() {defaultValue=0.25, parameterName=ErosionMaxRadius}, Alpha=Saturate(None=ScalarParameter() {parameterName=Erode})) {constB=0.25}
  n39 = Power(
   Base=Add(A=@n30, B=@n35)
   Exp=LinearInterpolate(B=ScalarParameter() {parameterName=CenterPower}, Alpha=SphereMask(A=@n36, B=@n37, Radius=@n38) {attenuationRadius=0.25, hardnessPercent=0}) {constA=0.55, constB=5}
  )
  n40 = ScalarParameter() {defaultValue=1, parameterName=AngleFade}
  n41 = Saturate(None=Multiply(A=MaterialFunctionCall("FadeSensitivity (optional)"=@n40) {materialFunction=&FadeBasedOnViewAngle}.Result) {constB=2})
  n42 = Multiply(
   A=MaterialFunctionCall(Distance=ScalarParameter() {defaultValue=800, parameterName=NearFadeDIst}) {materialFunction=&Near_Camera_Fade}.Result
   B=DepthFade(FadeDistance=ScalarParameter() {defaultValue=200, parameterName=DepthFade})
  )
  n43 = Multiply(
   A=Multiply(A=Multiply(A=@n39) {constB=2}, B=@n34)
   B=ShadingPathSwitch(Default=Multiply(A=@n41, B=@n42), Mobile=Constant() {r=0.7}) {desc=Migrated from FeatureLevelSwitch}
  )
  n44 = Multiply(A=@n43) {constB=0.3}
  Multiply(
   A=StaticSwitchParameter(
    True=MaterialFunctionCall(Color_Morning=@n44, Color_Day=@n43, Color_Evening=@n44, Color_Night=@n44) {materialFunction=&Color_Time_of_Day}.Result
    False=@n43
   ) {parameterName=Use color time of day}
   B=ScalarParameter() {defaultValue=1, parameterName=FinalOpacity}
  )
 ]
}
